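# Niagara system for the Slimefire vehicle boost cosmetic, emitting a trail and exhaust effects from filtered skeletal mesh sockets, shifted by a per-car offset.
# A six-renderer ribbon emitter layers the fading trail with curve-driven colors and widths, while a sprite emitter adds short-lived, shrinking particles streaming backward.
# A looping local-space emitter merges sprite and mesh elements, using spawn groups to select size and renderer visibility.
# All three emitters apply painted-variant color overrides.
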NiagaraSystem /VehicleCosmetics/Boosts/Boost_Slimefire/FX/NS_Boost_Slimefire
dependencies {
 usedDataInterfaces [
  &NiagaraDataInterfaceColorCurve
  &NiagaraDataInterfaceCurve
  &NiagaraDataInterfaceArrayFloat
  &NiagaraDataInterfaceArrayFloat3
  &NiagaraDataInterfaceSkeletalMesh
 ]
 usedDynamicInputs []
 usedModules [
  &self:SystemSpawnScript.SystemScriptSource.SystemScriptGraph.NiagaraNodeAssignment_0.SetVariables_#1
  &SystemState
 ]
}
emitters {
 NE_VC_Boost_Ribbon_Parent {
  emitterSpawnScript [/SetVariables(
    Emitter.Curve for Colors=DataInterface() {propertyValues {
      ExposedName=MainTrailGradient
      bExposeCurve=true
      AlphaCurve {keys [<at 0.9003 1>]}
      BlueCurve {keys [<at 0.0949 0.1787>,<at 0.9687 0>]}
      GreenCurve {keys [<at 0.0949 0>,<at 0.9687 0.9333>]}
      RedCurve {keys [<at 0.0949 3>,<at 0.9687 2>]}
     }}
    Emitter.bool:NiagaraBool=&InvertBool(Boolean:NiagaraBool=$User.bIsPainted)
   ) {isSetParameters=true}]
  emitterUpdateScript [
   &EmitterState()
   &VC_FilteredSocketCount()
   &VC_RateQualityHandler()
   &VC_SpawnRibbon_v1(Number of Ribbons=$Emitter.PortCount, SpawnRate=$Emitter.Spawn Rate)
  ]
  particleSpawnScript [
   &VC_PaintedVariantsOverride(
    Black=<rgba 0.018 0.018 0.018 1>
    BurntSienna=<rgba 0.1 0.02 0 1>
    Cobalt=<rgba 0.04 0.076 0.4 1>
    Crimson=<rgba 0.2 0 0.01 1>
    DefaultColor=<rgba 0.7 0.7 0.7 1>
    ForestGreen=<rgba 0 0.2 0.0101 1>
    Grey=<rgba 0.0875 0.0917 0.1 1>
    Lime=<rgba 0.1632 0.32 0.0064 1>
    Orange=<rgba 0.365 0.0456 0 1>
    Pink=<rgba 0.34 0.085 0.34 1>
    Purple=<rgba 0.0568 0.022 0.44 1>
    R1 Black:LinearColor=<rgba 0.038 0.038 0.038 1>
    R1 BurntSienna:LinearColor=<rgba 0.19 0.038 0 1>
    R1 Cobalt:LinearColor=<rgba 0.035 0.0682 0.7 1>
    R1 Crimson:LinearColor=<rgba 0.33 0 0 1>
    R1 DefaultColor=<rgba 0.0815 0.6215 0.7448 1>
    R1 ForestGreen:LinearColor=<rgba 0.023 0.33 0 1>
    R1 Grey:LinearColor=<rgba 0.1145 0.114 0.12 1>
    R1 Lime:LinearColor=<rgba 0.1952 0.33 0.0066 1>
    R1 Orange:LinearColor=<rgba 0.5 0.125 0 1>
    R1 Pink:LinearColor=<rgba 0.67 0.178 0.67 1>
    R1 Purple:LinearColor=<rgba 0.1432 0.0234 0.78 1>
    R1 Saffron:LinearColor=<rgba 0.7656 0.6554 0.0306 1>
    R1 SkyBlue:LinearColor=<rgba 0.076 0.4522 0.76 1>
    R1 TitaniumWhite:LinearColor=<rgba 0.57 0.57 0.57 1>
    R2 Black:LinearColor=<rgba 0.0885 0.0885 0.0885 1>
    R2 BurntSienna:LinearColor=<rgba 0.3613 0.1233 0.0542 1>
    R2 Cobalt:LinearColor=<rgba 0.1888 0.2558 0.8906 1>
    R2 Crimson:LinearColor=<rgba 0.55 0.0055 0.0203 1>
    R2 DefaultColor:LinearColor=<rgba 1 0.45 0 1>
    R2 ForestGreen:LinearColor=<rgba 0 0.55 0.0278 1>
    R2 Grey:LinearColor=<rgba 0.3125 0.3125 0.3125 1>
    R2 Lime:LinearColor=<rgba 0.3069 0.526 0.0877 1>
    R2 Orange:LinearColor=<rgba 0.6823 0.2207 0.0068 1>
    R2 Pink:LinearColor=<rgba 0.7188 0.2875 0.7187 1>
    R2 Purple:LinearColor=<rgba 0.2972 0.188 1 1>
    R2 Saffron:LinearColor=<rgba 0.7 0.5833 0 1>
    R2 SkyBlue:LinearColor=<rgba 0 0.4852 0.7292 1>
    R2 TitaniumWhite:LinearColor=<rgba 0.6615 0.6615 0.6615 1>
    R3 Black:LinearColor=<rgba 0.1406 0.1406 0.1406 1>
    R3 BurntSienna:LinearColor=<rgba 0.5312 0.1356 0.0437 2>
    R3 Cobalt:LinearColor=<rgba 0.1906 0.1324 1 2>
    R3 Crimson:LinearColor=<rgba 0.6875 0 0.0116 2>
    R3 DefaultColor:LinearColor=<rgba 0 1.248 2 1>
    R3 ForestGreen:LinearColor=<rgba 0 0.7083 0.0358 2>
    R3 Grey:LinearColor=<rgba 0.1562 0.1562 0.1562 2>
    R3 Lime:LinearColor=<rgba 0.3437 0.6875 0 3>
    R3 Orange:LinearColor=<rgba 1 0.2068 0 3>
    R3 Pink:LinearColor=<rgba 0.5052 0.1816 0.3875 5>
    R3 Purple:LinearColor=<rgba 0.1513 0.0389 0.6795 5>
    R3 Saffron:LinearColor=<rgba 0.7969 0.6109 0 5>
    R3 SkyBlue:LinearColor=<rgba 0 0.4189 0.8333 5>
    R3 TitaniumWhite:LinearColor=<rgba 1 1 1 2>
    Saffron=<rgba 0.4 0.3281 0.008 1>
    SkyBlue=<rgba 0.0921 0.2783 0.6 1>
    TitaniumWhite=<rgba 0.375 0.4045 0.45 1>
    Use R1 Painted Overrides=true
    Use R2 Painted Overrides=true
    Use R3 Painted Overrides=true
   )
   &VC_InitializeRibbon_v1(
    Color=$Particles.Color
    Custom Offset=true
    Enable Skeletal Position=true
    Lifetime=0.8
    Local Vector Offset=&AddVectorToPosition(Position=<-40 0 0>, Vector=$User.DTCarOffset)
    Offset Amount=1
    Ribbon Width=70
   )
   &AddVelocity(Velocity=<-1500 0 0>) {module=AddVelocity001}
  ]
  particleUpdateScript [
   &ParticleState()
   &InheritSourceMovement(
    Applied Position Scale=&Multiply_VectorByFloat(Vector=<1 1 1>, Float=1)
    ApplyForce=false
    ApplyPosition=true
    ApplyPositionFalloff=false
   )
   &SolveForcesAndVelocity(Clamp Velocity=false, Limit Acceleration=false)
   &VC_UpdateRibbon_v1(Enable Skeletal Position=true)
   /SetVariables(
    Particles.ScaleR0:LinearColor=&VC_DynamicScaleColor(
     Color to Scale:LinearColor=$Particles.Color
     RGB Scale:Vector3f=<1 1 1>
     Alpha Scale:NiagaraFloat=&FloatFromCurve(FloatCurve=DataInterface() {propertyValues {Curve {
         defaultValue=1
         keys [
          {arriveTangent=-0.2271, interpMode=RCIM_Cubic, leaveTangent=-0.2271, tangentMode=RCTM_User, time=0, value=1}
          {arriveTangent=-0.1203, interpMode=RCIM_Cubic, leaveTangent=-0.1203, tangentMode=RCTM_User, time=0.8, value=0}
         ]
        }}})
    )
    Particles.ScaleR1:LinearColor=&VC_DynamicScaleColor(
     Color to Scale:LinearColor=$Particles.R1 Color
     RGB Scale:Vector3f=<1 1 1>
     Alpha Scale:NiagaraFloat=&FloatFromCurve(FloatCurve=DataInterface() {propertyValues {Curve {
         defaultValue=0
         keys [
          {interpMode=RCIM_Cubic, tangentMode=RCTM_User, time=0, value=1}
          {arriveTangent=-2.8862, interpMode=RCIM_Cubic, leaveTangent=-2.8862, tangentMode=RCTM_User, time=0.8, value=0}
         ]
        }}})
    )
    Particles.ScaleR2:LinearColor=&VC_DynamicScaleColor(
     Color to Scale:LinearColor=$Particles.R2 Color
     RGB Scale:Vector3f=&VectorFromFloat(Value=&RandomRangeFloat(
       Minimum=&FloatFromCurve(FloatCurve=DataInterface() {propertyValues {Curve {
           defaultValue=1
           keys [
            {interpMode=RCIM_Cubic, tangentMode=RCTM_User, time=0, value=0.5}
            {interpMode=RCIM_Cubic, tangentMode=RCTM_User, time=0.5, value=1.5}
            {tangentMode=RCTM_User, time=1, value=0.5}
           ]
          }}})
       Maximum=&FloatFromCurve(FloatCurve=DataInterface() {propertyValues {Curve {
           defaultValue=1
           keys [
            {interpMode=RCIM_Cubic, tangentMode=RCTM_User, time=0, value=1.5}
            {interpMode=RCIM_Cubic, tangentMode=RCTM_User, time=0.5, value=0.5}
            {tangentMode=RCTM_User, time=1, value=1.5}
           ]
          }}})
      ))
     Alpha Scale:NiagaraFloat=1
    )
    Particles.WidthCore:NiagaraFloat=&Multiply_Float(
     A=50
     B=&FloatFromCurve(FloatCurve=DataInterface() {propertyValues {Curve {
         defaultValue=1
         keys [
          {interpMode=RCIM_Cubic, tangentMode=RCTM_User, time=0.0595, value=1.0196}
          {arriveTangent=-2.4162, interpMode=RCIM_Cubic, leaveTangent=-2.4162, tangentMode=RCTM_User, time=0.1, value=0.52}
          {arriveTangent=-1.5805, interpMode=RCIM_Cubic, leaveTangent=-1.5805, tangentMode=RCTM_User, time=0.15, value=0}
         ]
        }}})
    )
    Particles.WidthOutsideFire:NiagaraFloat=&Multiply_Float(
     A=50
     B=&FloatFromCurve(FloatCurve=DataInterface() {propertyValues {Curve {
         defaultValue=1
         keys [
          {arriveTangent=-2.5064, interpMode=RCIM_Cubic, leaveTangent=-2.5064, tangentMode=RCTM_User, time=0, value=1.5}
          {interpMode=RCIM_Cubic, tangentMode=RCTM_User, time=0.4, value=0.9858}
          {arriveTangent=-4.821, interpMode=RCIM_Cubic, leaveTangent=-4.821, tangentMode=RCTM_User, time=0.8, value=0.0517}
         ]
        }}})
    )
   ) {isSetParameters=true}
   &ScaleRibbonWidth(Ribbon Width Scale=&FloatFromCurve(FloatCurve=DataInterface() {propertyValues {Curve {
        defaultValue=1
        keys [
         {interpMode=RCIM_Cubic, tangentMode=RCTM_User, time=0.6, value=1}
         {arriveTangent=-13.3612, interpMode=RCIM_Cubic, leaveTangent=-13.3612, tangentMode=RCTM_User, time=0.8, value=0.2}
        ]
       }}}))
   &ScaleColor(
    Scale Alpha=&FloatFromCurve(FloatCurve=DataInterface() {propertyValues {Curve {
        defaultValue=1
        keys [
         {arriveTangent=7.1629, interpMode=RCIM_Cubic, leaveTangent=7.1629, tangentMode=RCTM_User, time=0.03, value=0}
         {interpMode=RCIM_Cubic, tangentMode=RCTM_User, time=0.2, value=1}
        ]
       }}})
    ScaleA=true
    ScaleRGB=false
   )
   &DynamicMaterialParameters(
    Module.Index3Float0Write=true
    Module.Index3Float1Write=true
    Module.Index3Float2Write=true
    Module.Index3Float3Write=true
    Write Parameter Index 0=::NewEnumerator0
    Write Parameter Index 3=::NewEnumerator1
   )
  ]
  rendererClasses [&NiagaraRibbonRendererProperties]
  renderers [
   {rendererClass=&NiagaraRibbonRendererProperties}
   {rendererClass=&NiagaraRibbonRendererProperties, rendererIndex=1}
   {rendererClass=&NiagaraRibbonRendererProperties, rendererIndex=2}
   {rendererClass=&NiagaraRibbonRendererProperties, rendererIndex=3}
   {rendererClass=&NiagaraRibbonRendererProperties, rendererIndex=4}
   {rendererClass=&NiagaraRibbonRendererProperties, rendererIndex=5}
  ]
 }
 NE_VC_Boost_Sprite_Parent002 {
  emitterUpdateScript [
   &EmitterState()
   &VC_FilteredSocketCount()
   &SpawnRate(SpawnRate=&RandomRangeFloat(Minimum=18, Maximum=25), Use Spawn Probability=false)
  ]
  particleSpawnScript [
   &VC_PaintedVariantsOverride(
    Black=<rgba 0.1458 0.1458 0.1458 5>
    BurntSienna=<rgba 0.3594 0.111 0.0443 5>
    Cobalt=<rgba 0.1771 0.3121 1 5>
    Crimson=<rgba 0.4635 0 0.0126 5>
    DefaultColor=<rgba 2 1.0403 0 2.5>
    ForestGreen=<rgba 0 0.474 0.0239 5>
    Grey=<rgba 0.2656 0.2656 0.2656 5>
    Lime=<rgba 0.3646 0.7292 0 5>
    Pink=<rgba 1 0.3542 0.7652 5>
    Purple=<rgba 0.5056 0.1615 1 5>
    Saffron=<rgba 1 0.6902 0 5>
    SkyBlue=<rgba 0 0.6169 0.9271 5>
    TitaniumWhite=<rgba 0.8698 0.8698 0.8698 5>
   )
   &InitializeParticle(
    Color Mode=::NewEnumerator0
    Lifetime Max=0.4
    Lifetime Min=0.35
    Lifetime Mode=::NewEnumerator1
    Sprite Size Mode=::NewEnumerator1
    Uniform Sprite Size Max=25
    Uniform Sprite Size Min=20
    UsePositionOffset=false
    Write Material Random=true
   )
   &PartitionParticles(Number Of Partitions=$Emitter.PortCount)
   &SkeletalMeshLocation(
    Bone / Socket Index=$Particles.PartitionParticles.Partition
    Bone / Socket Sampling Mode=::NewEnumerator5
    MaskSpawnGroup=false
    Mesh Sampling Type=::NewEnumerator6
    Sampled Position Offset=&AddVectorToPosition(Position=<-80 0 0>, Vector=$User.DTCarOffset)
    Skeletal Mesh=$System.SkeletalMesh
   )
   &AddVelocity(Velocity=<-400 0 0>)
   &ShapeLocation(
    Box Midpoint=<1 0.5 0.5>
    Box Size=<50 50 50>
    MaskSpawnGroup=false
    Override Seed=false
    Shape Primitive=::NewEnumerator2
    Surface Only=false
    UseOwnerScale=true
   )
  ]
  particleUpdateScript [
   &ParticleState()
   &ScaleSpriteSize(Uniform Curve Sprite Scale=DataInterface() {propertyValues {Curve {
       defaultValue=0
       keys [{interpMode=RCIM_Cubic, tangentMode=RCTM_User, time=0, value=1}, {interpMode=RCIM_Cubic, tangentMode=RCTM_User, time=1, value=0}]
      }}})
   &SolveForcesAndVelocity(Clamp Velocity=false, Limit Acceleration=false)
  ]
  rendererClasses [&NiagaraSpriteRendererProperties]
  renderers [{rendererClass=&NiagaraSpriteRendererProperties}]
 }
 Sprite_Local_Space {
  emitterSpawnScript [/SetVariables(
    Emitter.Curve for Colors=DataInterface() {propertyValues {
      ExposedName=MainTrailGradient
      bExposeCurve=true
      AlphaCurve {keys [<at 0.506 1>]}
      BlueCurve {keys [<at 0.8547 0>,<at 0.9718 0>]}
      GreenCurve {keys [<at 0.8547 0.0514>,<at 0.9718 0.7157>]}
      RedCurve {keys [<at 0.8547 1.5>,<at 0.9718 1>]}
     }}
    Emitter.FloatIsPainted=HlslExpression() {hlslExpression="User.ColorVariantID == 0 ? 0 : 1"}
   ) {isSetParameters=true}]
  emitterUpdateScript [
   &EmitterState(
    Inactive Response=::NewEnumerator1
    Life Cycle Mode=::NewEnumerator1
    Loop Behavior=::NewEnumerator1
    Loop Duration Mode=::NewEnumerator1
    UseLoopDelay=false
   )
   &VC_FilteredSocketCount()
   &SpawnBurst_Instantaneous(Spawn Count=2, Use Spawn Probability=false, UseLoopCountLimit=false)
   &SpawnBurst_Instantaneous(Spawn Count=2, Spawn Group=1, Use Spawn Probability=false, UseLoopCountLimit=false) {module=SpawnBurst_Instantaneous001}
   &SpawnBurst_Instantaneous(Spawn Count=2, Spawn Group=2, Use Spawn Probability=false, UseLoopCountLimit=false) {module=SpawnBurst_Instantaneous002}
   &SpawnBurst_Instantaneous(Spawn Count=2, Spawn Group=3, Use Spawn Probability=false, UseLoopCountLimit=false) {module=SpawnBurst_Instantaneous003}
   &SpawnRate(SpawnRate=3, Use Spawn Probability=false) {enabled=false}
  ]
  particleSpawnScript [
   &VC_PaintedVariantsOverride(
    Black=<rgba 0.0312 0.0312 0.0312 1>
    BurntSienna=<rgba 0.15 0.0431 0.0075 1>
    Cobalt=<rgba 0.0198 0.0758 0.5 1>
    Crimson=<rgba 0.4479 0.0134 0.0252 1>
    DefaultColor=<rgba 0.8125 0.2634 0.0804 1>
    ForestGreen=<rgba 0.1116 0.45 0.0225 1>
    Grey=<rgba 0.1198 0.1198 0.1198 1>
    Lime=<rgba 0.4333 0.65 0 1>
    Orange=<rgba 0.65 0.1566 0.0065 1>
    Pink=<rgba 0.7 0.154 0.427 1>
    Purple=<rgba 0.1599 0.0855 0.7031 1>
    R1 Black:LinearColor=<rgba 0.125 0.125 0.125 1>
    R1 BurntSienna:LinearColor=<rgba 0.4583 0.149 0.0458 1>
    R1 Cobalt:LinearColor=<rgba 0.2362 0.4799 0.875 1>
    R1 Crimson:LinearColor=<rgba 0.6 0.18 0.25 1>
    R1 DefaultColor=<rgba 1.375 1.5 0 1>
    R1 ForestGreen:LinearColor=<rgba 0.2625 0.9 0.135 1>
    R1 Grey:LinearColor=<rgba 0.526 0.526 0.526 1>
    R1 Lime:LinearColor=<rgba 0.5 0.8 0.08 1>
    R1 Orange:LinearColor=<rgba 0.7656 0.3169 0.0766 1>
    R1 Pink:LinearColor=<rgba 1.2 0.36 0.78 1>
    R1 Purple:LinearColor=<rgba 0.4368 0.2198 0.5781 1>
    R1 Saffron:LinearColor=<rgba 0.8 0.66 0.24 1>
    R1 SkyBlue:LinearColor=<rgba 0.225 1.0734 1.5 1>
    R1 TitaniumWhite:LinearColor=<rgba 1.5 1.5 1.5 1>
    R2 Black:LinearColor=<rgba 0.055 0.055 0.055 1>
    R2 BurntSienna:LinearColor=<rgba 0.27 0.054 0 1>
    R2 Cobalt:LinearColor=<rgba 0.07 0.133 0.7 1>
    R2 Crimson:LinearColor=<rgba 0.5 0 0 1>
    R2 ForestGreen:LinearColor=<rgba 0.024 0.48 0 1>
    R2 Grey:LinearColor=<rgba 0.15 0.15 0.15 1>
    R2 Lime:LinearColor=<rgba 0.2917 0.5 0 1>
    R2 Orange:LinearColor=<rgba 0.9 0.165 0 1>
    R2 Pink:LinearColor=<rgba 1 0.25 0.9375 1>
    R2 Purple:LinearColor=<rgba 0.1925 0.05 1 1>
    R2 Saffron:LinearColor=<rgba 0.8 0.64 0 1>
    R2 SkyBlue:LinearColor=<rgba 0.11 0.6545 1.1 1>
    R2 TitaniumWhite:LinearColor=<rgba 0.65 0.65 0.65 1>
    R3 Black:LinearColor=<rgba 0.025 0.025 0.025 1>
    R3 BurntSienna:LinearColor=<rgba 0.2 0.04 0 1>
    R3 Cobalt:LinearColor=<rgba 0.0585 0.1176 0.65 1>
    R3 Crimson:LinearColor=<rgba 0.5 0 0 1>
    R3 DefaultColor:LinearColor=<rgba 1.5 1.5 1.5 1>
    R3 ForestGreen:LinearColor=<rgba 0.0175 0.35 0 1>
    R3 Grey:LinearColor=<rgba 0.14 0.14 0.14 1>
    R3 Lime:LinearColor=<rgba 0.2263 0.43 0.0086 1>
    R3 Orange:LinearColor=<rgba 0.67 0.0951 0.0067 1>
    R3 Pink:LinearColor=<rgba 0.55 0.149 0.55 1>
    R3 Purple:LinearColor=<rgba 0.1432 0.0853 0.78 1>
    R3 Saffron:LinearColor=<rgba 0.6 0.51 0.06 1>
    R3 SkyBlue:LinearColor=<rgba 0.18 0.605 1.2 1>
    R3 TitaniumWhite:LinearColor=<rgba 0.738 0.7517 0.82 1>
    Saffron=<rgba 0.6 0.4575 0.03 1>
    SkyBlue=<rgba 0.1185 0.4545 0.79 1>
    TitaniumWhite=<rgba 1 1 1 1>
    Use R1 Painted Overrides=true
    Use R2 Painted Overrides=true
    Use R3 Painted Overrides=true
   )
   &InitializeParticle(
    Color=$Particles.Color
    Lifetime=&SelectFloatFromArray(Direct Array Index=&SpawnGroup(), Float Selection Array=DataInterface() {propertyValues {FloatData [1,1,1,1]}})
    Mesh Scale Mode=::NewEnumerator3
    Mesh Uniform Scale=&SelectFloatFromArray(
     Direct Array Index=&SpawnGroup()
     Float Selection Array=DataInterface() {propertyValues {FloatData [0,0,2,1.9]}}
    )
    Sprite Size Mode=::NewEnumerator3
    Uniform Sprite Size=&SelectFloatFromArray(
     Direct Array Index=&SpawnGroup()
     Float Selection Array=DataInterface() {propertyValues {FloatData [35,30,0,0]}}
    )
    UsePositionOffset=false
    Write Material Random=true
   )
   &PartitionParticles(Number Of Partitions=$Emitter.PortCount)
   &SkeletalMeshLocation(
    Bone / Socket Index=$Particles.PartitionParticles.Partition
    Bone / Socket Sampling Mode=::NewEnumerator5
    MaskSpawnGroup=false
    Mesh Sampling Type=::NewEnumerator6
    Sampled Position Offset=&AddVectorToPosition(
     Position=&SelectVectorFromArray(
      Direct Array Index=&SpawnGroup()
      Vector Selection Array=DataInterface() {propertyValues {FloatData [<-60 0 0>, <-60 0 0>, <-60 0 0>, <-60 0 0>]}}
     )
     Vector=$User.DTCarOffset
    )
    Skeletal Mesh=$System.SkeletalMesh
   )
   &AddVelocity(Velocity=&SelectVectorFromArray(
     Direct Array Index=&SpawnGroup()
     Vector Selection Array=DataInterface() {propertyValues {FloatData [<0 0 0>, <0 0 0>]}}
    )) {enabled=false}
   &CameraOffset(Camera Offset Amount=/VehicleCosmetics/FX/Niagara/Modules/MergeEmitters/VC_FloatBySpawnGroup(
     SpawnGroup or ExecIndex:NiagaraInt32=0
     CustomInt:NiagaraInt32=0
     MaximumSpawnGroups:NiagaraInt32=4
     Float SG 0:NiagaraFloat=0
     Float SG 1:NiagaraFloat=0
     Float SG 2:NiagaraFloat=50
     Float SG 3:NiagaraFloat=0
     Float SG 4:NiagaraFloat=0
     Float SG 5:NiagaraFloat=0
    ))
  ]
  particleUpdateScript [
   &ParticleState(Kill Particles When Lifetime Has Elapsed=false, Loop Particles Lifetime=true)
   &SolveForcesAndVelocity(Clamp Velocity=false, Limit Acceleration=false)
   /SetVariables(Particles.Visibility:NiagaraInt32=&SpawnGroup()) {isSetParameters=true}
   &ScaleColor(ScaleA=true, ScaleRGB=true)
  ]
  rendererClasses [&NiagaraSpriteRendererProperties, &NiagaraMeshRendererProperties]
  renderers [
   {rendererClass=&NiagaraSpriteRendererProperties}
   {rendererClass=&NiagaraMeshRendererProperties, rendererIndex=1}
   {rendererClass=&NiagaraMeshRendererProperties, rendererIndex=2}
   {rendererClass=&NiagaraSpriteRendererProperties, rendererIndex=3}
  ]
 }
}
summary {
 systemName=NS_Boost_Slimefire
 emitters [
  {emitterName=NE_VC_Boost_Ribbon_Parent, rendererClasses [&NiagaraRibbonRendererProperties]}
  {emitterName=NE_VC_Boost_Sprite_Parent002, rendererClasses [&NiagaraSpriteRendererProperties]}
  {
   emitterName=Sprite_Local_Space
   rendererClasses [&NiagaraSpriteRendererProperties, &NiagaraMeshRendererProperties]
  }
 ]
 userVariables [
  User.bIsFrontendPreview {type=NiagaraBool}
  User.bIsPainted {type=NiagaraBool}
  User.ColorVariantID {type=NiagaraInt32}
  User.DTCarOffset {type=Vector3f, description="", defaultValue {struct=Vector3f, value=<0 0 0>}}
  User.Intensity {type=NiagaraFloat, defaultValue=0, description=""}
  User.PaintedColor {type=LinearColor, description="", defaultValue {struct=LinearColor, value=<rgba 1 1 1 1>}}
  User.PaintedScalar {type=NiagaraFloat, defaultValue=1, description=""}
  User.PortCount {type=NiagaraInt32, defaultValue=2}
 ]
}
