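# Translucent, unlit, two-sided material for Niagara mesh particles, apparently the cone meshes of the Slimefire vehicle boost.
# It builds two looks, Distortion and BasicMat, from panning cloud noise plus UV gradients shaped by SmoothStep and Power, blended by the "Distortion=0" scalar without warping UVs.
# Each blends parameter colors or, via "IsPainted=0", desaturated color tinted by particle color, into MF_FX_EmissiveAdjustments.
# Opacity combines particle alpha with the noise masks, plus an optional U-edge mask on BasicMat.
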Material /VehicleCosmetics/Boosts/Boost_Slimefire/Materials/M_Slimefire_Cones
graph {
 properties {
  bAutomaticallySetUsageInEditor=false
  bUsedWithNiagaraMeshParticles=true
  blendMode=BLEND_Translucent
  shadingModel=MSM_Unlit
  twoSided=true
  usedShadingModels=Unlit
 }
 nodes [
  n1 = ParticleColor()
  n2 = TextureCoordinate()
  n3 = AppendVector(
   A=ScalarParameter() {defaultValue=1, group=Distortion, parameterName=DistortionTilingU}
   B=ScalarParameter() {defaultValue=1, group=Distortion, parameterName=DistortionTilingV}
  )
  n4 = Time()
  n5 = ScalarParameter() {group=Distortion, parameterName=DistortionSpeed}
  n6 = TextureSampleParameter2D(UVs=Panner(Coordinate=Multiply(A=@n2, B=@n3), Time=Multiply(A=@n4, B=@n5)) {speedX=0.05, speedY=0.8}) {
   group=Distortion
   parameterName=DisstortTexture
   samplerType=SAMPLERTYPE_Normal
   texture=/DealershipSharedLibrary/FX/Textures/Noise/T_VCVFX_CloudNoise_N
  }
  n7 = ComponentMask(None=TextureCoordinate()) {g=true}
  n8 = Multiply(A=ScalarParameter() {defaultValue=1, group=Distortion, parameterName=LinearMaskInverse2}, B=@n7)
  n9 = Multiply(A=OneMinus(None=@n7), B=ScalarParameter() {defaultValue=1, group=Distortion, parameterName=LinearMask2})
  n10 = SmoothStep(
   Min=ScalarParameter() {group=Distortion, parameterName=DistStepMin}
   Max=ScalarParameter() {defaultValue=1, group=Distortion, parameterName=DistStepMax}
   Value=Add(A=@n6.R, B=Multiply(A=@n8, B=@n9))
  )
  n11 = ScalarParameter() {defaultValue=1, group=Distortion, parameterName=DistContrastPower}
  MaterialExpressionMeshEmitterVertexColor_10 = ParticleColor()
  n12 = TextureCoordinate()
  n13 = AppendVector(
   A=ScalarParameter() {defaultValue=1, group=BasicMat, parameterName=TilingU}
   B=ScalarParameter() {defaultValue=1, group=BasicMat, parameterName=TilingV}
  )
  n14 = Time()
  n15 = ScalarParameter() {defaultValue=1, group=BasicMat, parameterName=Speed}
  n16 = TextureSampleParameter2D(UVs=Panner(Coordinate=Multiply(A=@n12, B=@n13), Time=Multiply(A=@n14, B=@n15)) {speedX=0.05, speedY=0.5}) {
   group=BasicMat
   parameterName=MainTexture
   samplerType=SAMPLERTYPE_Normal
   texture=&T_VCVFX_CloudNoise_N
  }
  n17 = ComponentMask(None=TextureCoordinate()) {g=true}
  n18 = OneMinus(None=@n17)
  n19 = ScalarParameter() {defaultValue=1, group=BasicMat, parameterName=LinearMask}
  n20 = Add(
   A=@n16.R
   B=StaticSwitchParameter(True=Multiply(A=@n18, B=@n19), False=Multiply(A=@n17, B=@n19)) {defaultValue=true, parameterName=Inverse Gradient}
  )
  n21 = Power(
   Base=SmoothStep(
    Min=ScalarParameter() {group=BasicMat, parameterName=FadeStepMin}
    Max=ScalarParameter() {defaultValue=1, group=BasicMat, parameterName=FadeStepMax}
    Value=@n20
   )
   Exp=ScalarParameter() {defaultValue=1, group=BasicMat, parameterName=ContrastPowerFade}
  )
  n22 = Constant() {r=1}
  n23 = ComponentMask(None=TextureCoordinate()) {r=true}
  n24 = Multiply(
   A=Multiply(A=Power(Base=@n23) {constExponent=1.5}, B=OneMinus(None=@n23))
   B=ScalarParameter() {defaultValue=25, group=BasicMat, parameterName=MaskMultiplier}
  ) {constB=25}
  n25 = ScalarParameter() {defaultValue=1, group=BasicMat, parameterName="UseMask=1"}
  n26 = ScalarParameter() {parameterName="Distortion=0"}
  LinearInterpolate(
   A=Multiply(A=@n1.A, B=Power(Base=@n10, Exp=@n11))
   B=Multiply(A=Multiply(A=@MaterialExpressionMeshEmitterVertexColor_10.A, B=@n21), B=LinearInterpolate(A=@n22, B=@n24, Alpha=@n25))
   Alpha=@n26
  )
  n27 = ScalarParameter() {group=Distortion, parameterName=MainStepMin}
  n28 = ScalarParameter() {defaultValue=1, group=Distortion, parameterName=MainStepMax}
  n29 = TextureCoordinate()
  n30 = AppendVector(
   A=ScalarParameter() {defaultValue=1, group=Distortion, parameterName=MainTilingU}
   B=ScalarParameter() {defaultValue=1, group=Distortion, parameterName=MainTilingV}
  )
  n31 = Time()
  n32 = ScalarParameter() {defaultValue=1, group=Distortion, parameterName=MainSpeed}
  n33 = TextureSampleParameter2D(UVs=Panner(Coordinate=Multiply(A=@n29, B=@n30), Time=Multiply(A=@n31, B=@n32)) {speedX=0.05, speedY=0.8}) {
   group=Distortion
   parameterName=Main Texture
   samplerType=SAMPLERTYPE_Normal
   texture=&T_VCVFX_CloudNoise_N
  }
  n34 = ComponentMask(None=TextureCoordinate()) {g=true}
  n35 = Multiply(A=ScalarParameter() {defaultValue=1, group=Distortion, parameterName=LinearMaskInverse2}, B=@n34)
  n36 = Multiply(A=OneMinus(None=@n34), B=ScalarParameter() {defaultValue=1, group=Distortion, parameterName=LinearMask2})
  n37 = Add(A=@n33.RGB, B=Saturate(None=Multiply(A=@n35, B=@n36)))
  n38 = LinearInterpolate(
   A=VectorParameter() {group=Distortion, parameterName=SecondColor, defaultValue=<rgba 0.9387 0 0.0785 1>}.RGB
   B=VectorParameter() {group=Distortion, parameterName=FirstColor, defaultValue=<rgba 0 0.152 1 1>}.RGB
   Alpha=Power(Base=SmoothStep(Min=@n27, Max=@n28, Value=@n37), Exp=ScalarParameter() {defaultValue=1, group=Distortion, parameterName=MainContrastPower})
  )
  n39 = Multiply(A=@n38, B=ScalarParameter() {defaultValue=1, group=Distortion, parameterName=MainColorMultiply})
  n40 = Multiply(A=Desaturation(None=@n38), B=@n1.RGB)
  n41 = ScalarParameter() {defaultValue=1, parameterName="IsPainted=0"}
  n42 = VectorParameter() {group=BasicMat, parameterName=ColorDark, defaultValue=<rgba 0 0.2851 0.5 1>}
  n43 = ComponentMask(None=TextureCoordinate()) {r=true}
  n44 = Multiply(A=@n43, B=OneMinus(None=@n43))
  MaterialExpressionConstantClamp_21 = Saturate(None=Multiply(A=@n44, B=Constant() {r=2.5}))
  n45 = LinearInterpolate(
   A=@n42.RGB
   B=VectorParameter() {group=BasicMat, parameterName=MainColor, defaultValue=<rgba 0 1 0.8717 1>}.RGB
   Alpha=OneMinus(None=@MaterialExpressionConstantClamp_21)
  )
  MaterialExpressionConstantClamp_20 = Saturate(None=Multiply(A=@n44, B=Constant() {r=8}))
  n46 = ScalarParameter() {group=BasicMat, parameterName=StepMin}
  n47 = ScalarParameter() {defaultValue=1, group=BasicMat, parameterName=StepMax}
  n48 = LinearInterpolate(
   A=Multiply(A=ScalarParameter() {defaultValue=1, group=BasicMat, parameterName=Dark Multiplier}, B=@n42.RGB)
   B=LinearInterpolate(
    A=VectorParameter() {group=BasicMat, parameterName=LightColor, defaultValue=<rgba 0 0.1082 1 1>}.RGB
    B=@n45
    Alpha=@MaterialExpressionConstantClamp_20
   )
   Alpha=Power(Base=SmoothStep(Min=@n46, Max=@n47, Value=@n20), Exp=ScalarParameter() {defaultValue=1, group=BasicMat, parameterName=ContrastPower})
  )
  n49 = Multiply(A=Desaturation(None=@n45), B=@MaterialExpressionMeshEmitterVertexColor_10.RGB)
  n50 = ScalarParameter() {defaultValue=1, parameterName="IsPainted=0"}
  MaterialFunctionCall(Emissive=LinearInterpolate(A=LinearInterpolate(A=@n39, B=@n40, Alpha=@n41), B=LinearInterpolate(A=@n48, B=@n49, Alpha=@n50), Alpha=@n26)) {materialFunction=&MF_FX_EmissiveAdjustments}
 ]
}
